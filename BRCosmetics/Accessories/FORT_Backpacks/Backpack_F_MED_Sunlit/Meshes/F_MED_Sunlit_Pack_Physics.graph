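# Physics asset for the Sunlit backpack cosmetic, providing bodies and joints for its dangling zipper pulls and charm.
# Two capsules and a box shape the pack body, while the small dyn_ bones use capsules, mostly paired with thin boxes.
# Zipper_01 joins the pack directly; BurrerFly hangs from Chain_01, Zipper_02, then the pack.
# Zippers swing mainly on one axis with almost no twist, the chain link moves more freely, and BurrerFly locks one swing axis.
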PhysicsAsset /BRCosmetics/Accessories/FORT_Backpacks/Backpack_F_MED_Sunlit/Meshes/F_MED_Sunlit_Pack_Physics
bodies {
 dyn_BurrerFly [
  {shapeType=Capsule, length=1.5071, radius=0.5, shapeName=dyn_BurrerFly_capsule, center=<1.0953 0.0053 0.0183>, rotation=<rot 0 90 90>}
  {shapeType=Box, extentX=4.4799, extentY=0.5, extentZ=4, shapeName=dyn_BurrerFly_box, center=<0.8489 0 0>}
 ]
 dyn_Chain_01 [{
   shapeType=Capsule
   length=0.505
   radius=0.505
   shapeName=dyn_Chain_01_capsule
   center=<0.8566 0.0109 -0.0071>
   rotation=<rot 0 -89.9632 -89.9004>
  }]
 dyn_Zipper_01 [
  {
   shapeType=Capsule
   length=1.1615
   radius=0.3
   shapeName=dyn_Zipper_01_capsule
   center=<1.8048 -0.0009 -0.119>
   rotation=<rot 0 -89.9152 -86.2547>
  }
  {
   shapeType=Box
   extentX=2.1
   extentY=1.3
   extentZ=0.2
   shapeName=dyn_Zipper_01_box
   center=<1.7446 0.0001 -0.1118>
   rotation=<rot -4.1216 0.0188 0>
  }
 ]
 dyn_Zipper_02 [
  {shapeType=Capsule, length=2, radius=0.2, shapeName=dyn_Zipper_02_capsule, center=<1.4882 -0.0052 0.0108>, rotation=<rot 0 89.8872 -89.7636>}
  {shapeType=Box, extentX=2, extentY=1.3, extentZ=0.2, shapeName=dyn_Zipper_02_box, center=<1.4229 0 0>}
 ]
 pack [
  {shapeType=Capsule, length=13.7839, radius=6.7142, center=<-0.7908 -4.1575 -0.3604>, rotation=<rot -2.1362 -88.6887 -30.9466>}
  {shapeType=Capsule, length=15.5963, radius=2.8711, shapeName=pack_capsule, center=<-6.5903 -9.1099 -1.2928>, rotation=<rot 34.3393 0 0>}
  {shapeType=Box, extentX=9.4859, extentY=0.5, extentZ=16.9952, shapeName=pack_box, center=<-7.8631 2.2433 -5.4571>}
 ]
}
constraints [
 {bone1Name=dyn_Zipper_01, bone2Name=pack, swing1LimitDegrees=4, swing2LimitDegrees=60, twistLimitDegrees=1}
 {bone1Name=dyn_BurrerFly, bone2Name=dyn_Chain_01, swing1LimitDegrees=15, swing1Motion=Locked, swing2LimitDegrees=0, twistLimitDegrees=32.28}
 {bone1Name=dyn_Chain_01, bone2Name=dyn_Zipper_02, swing1LimitDegrees=33.12, swing2LimitDegrees=17.28, twistLimitDegrees=33.48}
 {bone1Name=dyn_Zipper_02, bone2Name=pack, swing1LimitDegrees=5, swing2LimitDegrees=50.76, twistLimitDegrees=1}
]
